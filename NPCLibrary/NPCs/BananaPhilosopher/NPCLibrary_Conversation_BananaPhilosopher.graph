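# Conversation database for the Banana Philosopher NPC, serving as its greeting and top-level conversation hub.
# An entry point leads to one speech node with a default "NaNa" line and contextual variants for first-ever meeting, match conversation count and player outfit.
# Outfit reactions check cosmetic character IDs, with separate lines for banana skins and the philosopher's own skin.
# The speech then branches through link nodes to the remote Quests, BasicServices and BananaPhilosopher Dialogue entry points.
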ConversationDatabase /NPCLibrary/NPCs/BananaPhilosopher/NPCLibrary_Conversation_BananaPhilosopher
entry_points [{tag=Conversation.Entry.BananaPhilosopher, destinations [#1]}]
nodes [
 ConversationEntryPointNode_7 {
  id=#1
  class=ConversationEntryPointNode
  outputs [#2]
  properties {entryTag {tagName=Conversation.Entry.BananaPhilosopher}}
 }
 Speech_C_4 {
  id=#2
  class=Speech_C
  outputs [#3, #4, #5]
  properties {
   nodeName=""
   generalConfig {
    defaultMessage="NaNa, NaNa Na!"
    contextualMessages [
     {
      message=NaNa Na?
      contextRequirements [{requirement {class=FortControllerRequirement_IsFirstEverConversationWithNPC}}]
      priority {curve {rowName=Tandem.Speech.Priority.FirstEverInteraction}}
     }
     {
      message="NaNa, NaNa Na!"
      contextRequirements [{requirement {
         class=FortControllerRequirement_CheckNPCMatchConversationCount
         properties {nPCUniqueIdTag {tagName=AISpawnerData.Type.Tandem.BananaPhilosopher}}
        }}]
      priority {curve {rowName=None}}
     }
     {
      message=NaNa Na. Na?
      contextRequirements [{requirement {
         class=FortControllerRequirement_CheckNPCMatchConversationCount
         properties {nPCUniqueIdTag {tagName=AISpawnerData.Type.Tandem.BananaPhilosopher}}
        }}]
      priority {curve {rowName=None}}
     }
     {
      message="NaNaNa Na Na? Na!"
      contextRequirements [{requirement {
         class=FortControllerRequirement_CheckNPCMatchConversationCount
         properties {nPCUniqueIdTag {tagName=AISpawnerData.Type.Tandem.BananaPhilosopher}}
        }}]
      priority {curve {rowName=None}}
     }
     {
      message="You look well, young banana."
      contextRequirements [{requirement {
         class=FortControllerRequirement_HasCID
         properties {allowedCIds [
           &CID_349_Athena_Commando_M_Banana
           /BRCosmetics/Athena/Items/Cosmetics/Characters/CID_602_Athena_Commando_M_NanaSplit
           &CID_701_Athena_Commando_M_BananaAgent
           &CID_814_Athena_Commando_M_BananaSummer
           /BRCosmetics/Athena/Items/Cosmetics/Characters/CID_A_259_Athena_Commando_M_PeelyTech
           &CID_A_323_Athena_Commando_M_BananaWinter
           /BRCosmetics/Athena/Items/Cosmetics/Characters/CID_A_355_Athena_Commando_M_PeelyToon
           /BRCosmetics/Athena/Items/Cosmetics/Characters/CID_A_437_Athena_Commando_M_ChiselMashup
           &Character_BananaAdventure
           /BRCosmetics/Athena/Items/Cosmetics/Characters/CID_534_Athena_Commando_M_PeelyMech
           /BRCosmetics/Athena/Items/Cosmetics/Characters/Character_LilSplit_Sprinkles
          ]}
        }}]
      priority {curve {rowName=Tandem.Speech.Priority.OutfitReaction}}
     }
     {
      message=What cursed trick is the cave playing on me now??
      contextRequirements [{requirement {
         class=FortControllerRequirement_HasCID
         properties {allowedCIds [&Character_BananaPhilosopher]}
        }}]
      priority {curve {rowName=Tandem.Speech.Priority.OutfitReaction}}
     }
    ]
   }
  }
 }
 ConversationLinkNode_2 {
  id=#5
  class=ConversationLinkNode
  properties {remoteEntryTag {tagName=Conversation.Entry.BananaPhilosopher.Dialogue}}
 }
 ConversationLinkNode_0 {id=#4, class=ConversationLinkNode, properties {remoteEntryTag {tagName=Conversation.Entry.BasicServices}}}
 ConversationLinkNode_1 {id=#3, class=ConversationLinkNode, properties {remoteEntryTag {tagName=Conversation.Entry.NPCLibrary.Quests}}}
]
