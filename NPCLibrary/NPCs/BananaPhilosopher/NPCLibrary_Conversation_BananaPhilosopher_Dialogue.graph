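# Conversation database for the Banana Philosopher NPC's "Ask about..." dialogue, which speaks plainly to banana-outfit players and in "Na" gibberish to others.
# The opening line swaps to English through an outfit-based contextual message, using a shorter character list than the topic gates.
# Character requirements show Wisdom and Bananas topics to listed outfits, and inverted copies show two gibberish topics otherwise.
# Each topic plays three chained lines into a shared return-to-start node.
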ConversationDatabase /NPCLibrary/NPCs/BananaPhilosopher/NPCLibrary_Conversation_BananaPhilosopher_Dialogue
entry_points [{tag=Conversation.Entry.BananaPhilosopher.Dialogue, destinations [#1]}]
nodes [
 ConversationEntryPointNode_0 {
  id=#1
  class=ConversationEntryPointNode
  outputs [#2]
  properties {entryTag {tagName=Conversation.Entry.BananaPhilosopher.Dialogue}}
 }
 Speech_C_0 {
  id=#2
  class=Speech_C
  outputs [
   #3
   #4
   #5
   #6
   #7
  ]
  properties {
   nodeName=""
   softTaskIcon=&T_UI_NPC_question
   generalConfig {
    defaultMessage="NaNa, NaNaNaNaNa Na?"
    contextualMessages [{
      message="Well then, seedling, what would you like to know?"
      contextRequirements [{requirement {
         class=FortControllerRequirement_HasCID
         properties {allowedCIds [
           &CID_349_Athena_Commando_M_Banana
           /BRCosmetics/Athena/Items/Cosmetics/Characters/CID_602_Athena_Commando_M_NanaSplit
           &CID_701_Athena_Commando_M_BananaAgent
           &CID_814_Athena_Commando_M_BananaSummer
           /BRCosmetics/Athena/Items/Cosmetics/Characters/CID_A_259_Athena_Commando_M_PeelyTech
           &CID_A_323_Athena_Commando_M_BananaWinter
           /BRCosmetics/Athena/Items/Cosmetics/Characters/CID_A_355_Athena_Commando_M_PeelyToon
           /BRCosmetics/Athena/Items/Cosmetics/Characters/CID_A_437_Athena_Commando_M_ChiselMashup
           &Character_BananaAdventure
          ]}
        }}]
      priority {curve {rowName=Tandem.Speech.Priority.OutfitReaction}}
     }]
   }
  }
  sub_nodes [ConversationChoiceNode_5 {class=ConversationChoiceNode, properties {defaultChoiceDisplayText=Ask about...}}]
 }
 Speech_C_8 {
  id=#7
  class=Speech_C
  outputs [#8]
  properties {nodeName="", softTaskIcon=&T_UI_NPC_question, generalConfig {defaultMessage=NaNa Na NaNaNa.}}
  sub_nodes [
   ConversationChoiceNode_15 {class=ConversationChoiceNode, properties {defaultChoiceDisplayText=Na NaNa}}
   FortConversationRequirement_ControllerRequirement_0 {
    class=FortConversationRequirement_ControllerRequirement
    properties {
     nodeName=Require: Uses Character
     requirement {
      class=FortControllerRequirement_HasCID
      properties {
       bInverseCondition=true
       allowedCIds [
        &CID_349_Athena_Commando_M_Banana
        &CID_602_Athena_Commando_M_NanaSplit
        &CID_701_Athena_Commando_M_BananaAgent
        &CID_814_Athena_Commando_M_BananaSummer
        &CID_A_259_Athena_Commando_M_PeelyTech
        &CID_A_323_Athena_Commando_M_BananaWinter
        &CID_A_355_Athena_Commando_M_PeelyToon
        &CID_A_437_Athena_Commando_M_ChiselMashup
        &Character_BananaAdventure
        &Character_BananaPhilosopher
        /BRCosmetics/Athena/Items/Cosmetics/Characters/Character_LilSplit_Sprinkles
        /BRCosmetics/Athena/Items/Cosmetics/Characters/CID_534_Athena_Commando_M_PeelyMech
       ]
      }
     }
    }
   }
  ]
 }
 Speech_C_9 {
  id=#8
  class=Speech_C
  outputs [#9]
  properties {nodeName="", softTaskIcon=&T_UI_NPC_question, generalConfig {defaultMessage="NaNa Na, Na Na."}}
 }
 Speech_C_10 {
  id=#9
  class=Speech_C
  outputs [#10]
  properties {nodeName="", softTaskIcon=&T_UI_NPC_question, generalConfig {defaultMessage="Na NaNaNaNa, NaNaNa."}}
 }
 ReturnToStart_C_23 {
  id=#10
  class=ReturnToStart_C
  properties {softTaskIcon=&T_UI_NPC_Back}
  sub_nodes [ConversationChoiceNode_32 {class=ConversationChoiceNode, properties {defaultChoiceDisplayText=BACK}}]
 }
 Speech_C_4 {
  id=#6
  class=Speech_C
  outputs [#11]
  properties {nodeName="", softTaskIcon=&T_UI_NPC_question, generalConfig {defaultMessage=NaNaNa NaNa Na.}}
  sub_nodes [
   ConversationChoiceNode_2 {class=ConversationChoiceNode, properties {defaultChoiceDisplayText=NaNa}}
   FortConversationRequirement_ControllerRequirement_3 {
    class=FortConversationRequirement_ControllerRequirement
    properties {
     nodeName=Require: Uses Character
     requirement {
      class=FortControllerRequirement_HasCID
      properties {
       bInverseCondition=true
       allowedCIds [
        &CID_349_Athena_Commando_M_Banana
        &CID_602_Athena_Commando_M_NanaSplit
        &CID_701_Athena_Commando_M_BananaAgent
        &CID_814_Athena_Commando_M_BananaSummer
        &CID_A_259_Athena_Commando_M_PeelyTech
        &CID_A_323_Athena_Commando_M_BananaWinter
        &CID_A_355_Athena_Commando_M_PeelyToon
        &CID_A_437_Athena_Commando_M_ChiselMashup
        &Character_BananaAdventure
        &Character_BananaPhilosopher
        &Character_LilSplit_Sprinkles
        &CID_534_Athena_Commando_M_PeelyMech
       ]
      }
     }
    }
   }
  ]
 }
 Speech_C_5 {
  id=#11
  class=Speech_C
  outputs [#12]
  properties {nodeName="", softTaskIcon=&T_UI_NPC_question, generalConfig {defaultMessage="NaNa Na, NaNaNaNa Na."}}
 }
 Speech_C_6 {
  id=#12
  class=Speech_C
  outputs [#10]
  properties {nodeName="", softTaskIcon=&T_UI_NPC_question, generalConfig {defaultMessage="Na, NaNaNaNaNa, NaNa."}}
 }
 Speech_C_1 {
  id=#5
  class=Speech_C
  outputs [#13]
  properties {
   nodeName=""
   softTaskIcon=&T_UI_NPC_question
   generalConfig {defaultMessage=I understand you meet few of our kin here.}
  }
  sub_nodes [
   ConversationChoiceNode_1 {class=ConversationChoiceNode, properties {defaultChoiceDisplayText=Bananas}}
   FortConversationRequirement_ControllerRequirement_2 {
    class=FortConversationRequirement_ControllerRequirement
    properties {
     nodeName=Require: Uses Character
     requirement {
      class=FortControllerRequirement_HasCID
      properties {allowedCIds [
        &CID_349_Athena_Commando_M_Banana
        &CID_602_Athena_Commando_M_NanaSplit
        &CID_701_Athena_Commando_M_BananaAgent
        &CID_814_Athena_Commando_M_BananaSummer
        &CID_A_259_Athena_Commando_M_PeelyTech
        &CID_A_323_Athena_Commando_M_BananaWinter
        &CID_A_355_Athena_Commando_M_PeelyToon
        &CID_A_437_Athena_Commando_M_ChiselMashup
        &Character_BananaAdventure
        &Character_BananaPhilosopher
        &Character_LilSplit_Sprinkles
        &CID_534_Athena_Commando_M_PeelyMech
       ]}
     }
    }
   }
  ]
 }
 Speech_C_2 {
  id=#13
  class=Speech_C
  outputs [#14]
  properties {nodeName="", softTaskIcon=&T_UI_NPC_question, generalConfig {defaultMessage="It can be very lonely, yes."}}
 }
 Speech_C_3 {
  id=#14
  class=Speech_C
  outputs [#10]
  properties {
   nodeName=""
   softTaskIcon=&T_UI_NPC_question
   generalConfig {defaultMessage="Know that you can always talk to me when you feel alone, young seedling."}
  }
 }
 Back_C_0 {
  id=#4
  class=Back_C
  sub_nodes [ConversationChoiceNode_7 {class=ConversationChoiceNode, properties {defaultChoiceDisplayText=BACK}}]
 }
 Speech_C_7 {
  id=#3
  class=Speech_C
  outputs [#15]
  properties {
   nodeName=""
   softTaskIcon=&T_UI_NPC_question
   generalConfig {defaultMessage="True wisdom, seedling, is embracing your limits."}
  }
  sub_nodes [
   ConversationChoiceNode_0 {class=ConversationChoiceNode, properties {defaultChoiceDisplayText=Wisdom}}
   FortConversationRequirement_ControllerRequirement_1 {
    class=FortConversationRequirement_ControllerRequirement
    properties {
     nodeName=Require: Uses Character
     requirement {
      class=FortControllerRequirement_HasCID
      properties {allowedCIds [
        &CID_349_Athena_Commando_M_Banana
        &CID_602_Athena_Commando_M_NanaSplit
        &CID_701_Athena_Commando_M_BananaAgent
        &CID_814_Athena_Commando_M_BananaSummer
        &CID_A_259_Athena_Commando_M_PeelyTech
        &CID_A_323_Athena_Commando_M_BananaWinter
        &CID_A_355_Athena_Commando_M_PeelyToon
        &CID_A_437_Athena_Commando_M_ChiselMashup
        &Character_BananaAdventure
        &Character_BananaPhilosopher
        &Character_LilSplit_Sprinkles
        &CID_534_Athena_Commando_M_PeelyMech
       ]}
     }
    }
   }
  ]
 }
 Speech_C_15 {
  id=#15
  class=Speech_C
  outputs [#16]
  properties {
   nodeName=""
   softTaskIcon=&T_UI_NPC_question
   generalConfig {defaultMessage=We cannot know everything in this world.}
  }
 }
 Speech_C_14 {
  id=#16
  class=Speech_C
  outputs [#10]
  properties {
   nodeName=""
   softTaskIcon=&T_UI_NPC_question
   generalConfig {defaultMessage=But we can open our hearts to everything all the same.}
  }
 }
]
